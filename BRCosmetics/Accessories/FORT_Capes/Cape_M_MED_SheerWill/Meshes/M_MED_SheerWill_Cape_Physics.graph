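# Physics asset for the SheerWill cape cosmetic, defining the bodies and joints that drive its dynamic cape bones.
# Four chains (left, center-left, center-right, right) of eight capsules hang from small master anchors, with tight angular limits and wider swing at each top joint.
# Unconstrained capsules on torso, arms, legs, feet and rear loincloth bones provide collision shapes.
# An oversized root capsule sits below the origin, apparently as a floor.
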PhysicsAsset /BRCosmetics/Accessories/FORT_Capes/Cape_M_MED_SheerWill/Meshes/M_MED_SheerWill_Cape_Physics
bodies {
 backpack_root [{shapeType=Capsule, length=0, radius=3, shapeName=backpack_root_capsule}]
 calf_l [{shapeType=Capsule, length=37.2374, radius=16.5, shapeName=calf_l_capsule, center=<-13.3998 3.4567 0>, rotation=<rot 0 89.9995 -90>}]
 calf_r [{shapeType=Capsule, length=37.2374, radius=16.5, shapeName=calf_r_capsule, center=<13.3998 -3.4567 0>, rotation=<rot 0 89.9995 90>}]
 cape_top [{shapeType=Capsule, length=0.505, radius=0.505, shapeName=cape_top_capsule, rotation=<rot 0 -90 0>}]
 clavicle_l [{shapeType=Capsule, length=7, radius=6, shapeName=clavicle_l_capsule, center=<5.9369 0 -1.8368>, rotation=<rot -90 -0.0001 0>}]
 clavicle_r [{shapeType=Capsule, length=9, radius=8.5, shapeName=clavicle_r_capsule, center=<-5.9369 0 1.7331>, rotation=<rot 90 0 0>}]
 dyn_cape_cl_01 [{shapeType=Capsule, length=14.1977, radius=2.7476, shapeName=dyn_cape_01_capsule, center=<7.15 0 1>, rotation=<rot 90 0 0>}]
 dyn_cape_cl_02 [{shapeType=Capsule, length=15.7494, radius=2.7476, shapeName=dyn_cape_01_capsule, center=<7.15 0 1>, rotation=<rot 90 0 0>}]
 dyn_cape_cl_03 [{shapeType=Capsule, length=15.7494, radius=2.7476, shapeName=dyn_cape_01_capsule, center=<7.15 0 1>, rotation=<rot 90 0 0>}]
 dyn_cape_cl_04 [{shapeType=Capsule, length=15.7494, radius=2.7476, shapeName=dyn_cape_01_capsule, center=<7.15 0 1>, rotation=<rot 90 0 0>}]
 dyn_cape_cl_05 [{shapeType=Capsule, length=15.7494, radius=2.7476, shapeName=dyn_cape_01_capsule, center=<7.15 0 1>, rotation=<rot 90 0 0>}]
 dyn_cape_cl_06 [{shapeType=Capsule, length=15.7494, radius=2.7476, shapeName=dyn_cape_01_capsule, center=<7.15 0 1>, rotation=<rot 90 0 0>}]
 dyn_cape_cl_07 [{shapeType=Capsule, length=15.7494, radius=2.7476, shapeName=dyn_cape_01_capsule, center=<7.15 0 1>, rotation=<rot 90 0 0>}]
 dyn_cape_cl_08 [{shapeType=Capsule, length=15.7494, radius=3.1209, shapeName=dyn_cape_01_capsule, center=<10.1581 0 0.4>, rotation=<rot 90 0 0>}]
 dyn_cape_cl_MSTR [{shapeType=Capsule, length=0.505, radius=0.505, shapeName=dyn_cape_r_MSTR_capsule, rotation=<rot 0 -90 0>}]
 dyn_cape_cr_01 [{shapeType=Capsule, length=14.1977, radius=2.7476, shapeName=dyn_cape_01_capsule, center=<7.15 0 1>, rotation=<rot 90 0 0>}]
 dyn_cape_cr_02 [{shapeType=Capsule, length=15.7494, radius=2.7476, shapeName=dyn_cape_01_capsule, center=<7.15 0 1>, rotation=<rot 90 0 0>}]
 dyn_cape_cr_03 [{shapeType=Capsule, length=15.7494, radius=2.7476, shapeName=dyn_cape_01_capsule, center=<7.15 0 1>, rotation=<rot 90 0 0>}]
 dyn_cape_cr_04 [{shapeType=Capsule, length=15.7494, radius=2.7476, shapeName=dyn_cape_01_capsule, center=<7.15 0 1>, rotation=<rot 90 0 0>}]
 dyn_cape_cr_05 [{shapeType=Capsule, length=15.7494, radius=2.7476, shapeName=dyn_cape_01_capsule, center=<7.15 0 1>, rotation=<rot 90 0 0>}]
 dyn_cape_cr_06 [{shapeType=Capsule, length=15.7494, radius=2.7476, shapeName=dyn_cape_01_capsule, center=<7.15 0 1>, rotation=<rot 90 0 0>}]
 dyn_cape_cr_07 [{shapeType=Capsule, length=15.7494, radius=2.7476, shapeName=dyn_cape_01_capsule, center=<7.15 0 1>, rotation=<rot 90 0 0>}]
 dyn_cape_cr_08 [{shapeType=Capsule, length=15.7494, radius=3.1209, shapeName=dyn_cape_01_capsule, center=<10.1581 0 0.4>, rotation=<rot 90 0 0>}]
 dyn_cape_cr_MSTR [{shapeType=Capsule, length=0.505, radius=0.505, shapeName=dyn_cape_r_MSTR_capsule, rotation=<rot 0 -90 0>}]
 dyn_cape_l_01 [{shapeType=Capsule, length=14.1977, radius=2.7476, shapeName=dyn_cape_01_capsule, center=<7.15 0 1>, rotation=<rot 90 0 0>}]
 dyn_cape_l_02 [{shapeType=Capsule, length=15.7494, radius=2.7476, shapeName=dyn_cape_01_capsule, center=<7.15 0 1>, rotation=<rot 90 0 0>}]
 dyn_cape_l_03 [{shapeType=Capsule, length=15.7494, radius=2.7476, shapeName=dyn_cape_01_capsule, center=<7.15 0 1>, rotation=<rot 90 0 0>}]
 dyn_cape_l_04 [{shapeType=Capsule, length=15.7494, radius=2.7476, shapeName=dyn_cape_01_capsule, center=<7.15 0 1>, rotation=<rot 90 0 0>}]
 dyn_cape_l_05 [{shapeType=Capsule, length=15.7494, radius=2.7476, shapeName=dyn_cape_01_capsule, center=<7.15 0 1>, rotation=<rot 90 0 0>}]
 dyn_cape_l_06 [{shapeType=Capsule, length=15.7494, radius=2.7476, shapeName=dyn_cape_01_capsule, center=<7.15 0 1>, rotation=<rot 90 0 0>}]
 dyn_cape_l_07 [{shapeType=Capsule, length=15.7494, radius=2.7476, shapeName=dyn_cape_01_capsule, center=<7.15 0 1>, rotation=<rot 90 0 0>}]
 dyn_cape_l_08 [{shapeType=Capsule, length=15.7494, radius=3.1209, shapeName=dyn_cape_01_capsule, center=<10.1581 0 0.4>, rotation=<rot 90 0 0>}]
 dyn_cape_l_MSTR [{shapeType=Capsule, length=0.505, radius=0.505, shapeName=dyn_cape_r_MSTR_capsule, rotation=<rot 0 -90 0>}]
 dyn_cape_r_01 [{shapeType=Capsule, length=14.1977, radius=2.7476, shapeName=dyn_cape_01_capsule, center=<7.15 0 1>, rotation=<rot 90 0 0>}]
 dyn_cape_r_02 [{shapeType=Capsule, length=15.7494, radius=2.7476, shapeName=dyn_cape_01_capsule, center=<7.15 0 1>, rotation=<rot 90 0 0>}]
 dyn_cape_r_03 [{shapeType=Capsule, length=15.7494, radius=2.7476, shapeName=dyn_cape_01_capsule, center=<7.15 0 1>, rotation=<rot 90 0 0>}]
 dyn_cape_r_04 [{shapeType=Capsule, length=15.7494, radius=2.7476, shapeName=dyn_cape_01_capsule, center=<7.15 0 1>, rotation=<rot 90 0 0>}]
 dyn_cape_r_05 [{shapeType=Capsule, length=15.7494, radius=2.7476, shapeName=dyn_cape_01_capsule, center=<7.15 0 1>, rotation=<rot 90 0 0>}]
 dyn_cape_r_06 [{shapeType=Capsule, length=15.7494, radius=2.7476, shapeName=dyn_cape_01_capsule, center=<7.15 0 1>, rotation=<rot 90 0 0>}]
 dyn_cape_r_07 [{shapeType=Capsule, length=15.7494, radius=2.7476, shapeName=dyn_cape_01_capsule, center=<7.15 0 1>, rotation=<rot 90 0 0>}]
 dyn_cape_r_08 [{shapeType=Capsule, length=15.7494, radius=3.1209, shapeName=dyn_cape_01_capsule, center=<10.1581 0 0.4>, rotation=<rot 90 0 0>}]
 dyn_cape_r_MSTR [{shapeType=Capsule, length=0.505, radius=0.505, shapeName=dyn_cape_r_MSTR_capsule, rotation=<rot 0 -90 0>}]
 dyn_loincloth_back_2 [{shapeType=Capsule, length=13.0767, radius=12.6907, shapeName=dyn_loincloth_back_2_capsule, center=<0 0 -10.0285>, rotation=<rot 0 0 90>}]
 dyn_loincloth_back_3 [{shapeType=Capsule, length=13.0767, radius=12.6907, shapeName=dyn_loincloth_back_3_capsule, center=<0 0 -10.0285>, rotation=<rot 0 0 90>}]
 dyn_loincloth_back_4 [{shapeType=Capsule, length=13.0767, radius=12.6907, shapeName=dyn_loincloth_back_4_capsule, center=<0 0 -10.0285>, rotation=<rot 0 0 90>}]
 foot_l [{shapeType=Capsule, length=11, radius=19, shapeName=foot_l_capsule, center=<0.0002 14.4535 0>, rotation=<rot -0.0001 -0.001 -90>}]
 foot_r [{shapeType=Capsule, length=11, radius=19, shapeName=foot_r_capsule, center=<-0.0002 -14.4535 0>, rotation=<rot -0.0001 -0.001 90>}]
 hider [{shapeType=Capsule, length=0, radius=3, shapeName=hider_capsule}]
 pelvis [{shapeType=Capsule, length=7.3953, radius=15.2155, shapeName=pelvis_capsule}]
 root [{
   shapeType=Capsule
   length=232.7972
   radius=326.3409
   shapeName=root_capsule
   center=<0.0002 -73.1274 -326.3191>
   rotation=<rot 90 -179.9998 0>
  }]
 spine_01 [{shapeType=Capsule, length=1.5, radius=11.5, shapeName=spine_01_capsule}]
 spine_02 [{shapeType=Capsule, length=1, radius=12, shapeName=spine_02_capsule, center=<0 2.0607 0>}]
 spine_03 [{shapeType=Capsule, length=0, radius=14.8869, shapeName=spine_03_capsule, center=<0 3.2124 0>}]
 spine_04 [{shapeType=Capsule, length=1.5, radius=12.5, shapeName=spine_04_capsule, center=<0 1.7179 0>}]
 spine_05 [{shapeType=Capsule, length=11, radius=11, shapeName=spine_05_capsule}]
 thigh_l [{shapeType=Capsule, length=22.2502, radius=16.5, shapeName=thigh_l_capsule, center=<-16.4237 0 0>, rotation=<rot -90 -0.0001 0>}]
 thigh_r [{shapeType=Capsule, length=22.2502, radius=16.5, shapeName=thigh_r_capsule, center=<16.4237 0 0>, rotation=<rot 90 -0.0001 0>}]
 upperarm_l [{shapeType=Capsule, length=20.5, radius=8.5, shapeName=upperarm_l_capsule, center=<16.693 0 0>, rotation=<rot -90 0 0>}]
 upperarm_r [{shapeType=Capsule, length=21.5, radius=8.5, shapeName=upperarm_r_capsule, center=<-14.4918 -1.6465 0>, rotation=<rot 90 0.0002 0>}]
}
constraints [
 {bone1Name=dyn_cape_r_03, bone2Name=dyn_cape_r_02, swing1LimitDegrees=10, swing2LimitDegrees=15, twistLimitDegrees=8}
 {bone1Name=dyn_cape_r_06, bone2Name=dyn_cape_r_05, swing1LimitDegrees=10, swing2LimitDegrees=15, twistLimitDegrees=8}
 {bone1Name=dyn_cape_l_03, bone2Name=dyn_cape_l_02, swing1LimitDegrees=10, swing2LimitDegrees=15, twistLimitDegrees=8}
 {bone1Name=dyn_cape_l_06, bone2Name=dyn_cape_l_05, swing1LimitDegrees=10, swing2LimitDegrees=15, twistLimitDegrees=8}
 {bone1Name=dyn_cape_cr_03, bone2Name=dyn_cape_cr_02, swing1LimitDegrees=10, swing2LimitDegrees=15, twistLimitDegrees=8}
 {bone1Name=dyn_cape_cr_07, bone2Name=dyn_cape_cr_06, swing1LimitDegrees=10, swing2LimitDegrees=15, twistLimitDegrees=8}
 {bone1Name=dyn_cape_cl_03, bone2Name=dyn_cape_cl_02, swing1LimitDegrees=10, swing2LimitDegrees=15, twistLimitDegrees=8}
 {bone1Name=dyn_cape_cl_07, bone2Name=dyn_cape_cl_06, swing1LimitDegrees=10, swing2LimitDegrees=15, twistLimitDegrees=8}
 {bone1Name=dyn_cape_cl_01, bone2Name=dyn_cape_cl_MSTR, swing1LimitDegrees=10, swing2LimitDegrees=30, twistLimitDegrees=8}
 {bone1Name=dyn_cape_cl_02, bone2Name=dyn_cape_cl_01, swing1LimitDegrees=10, swing2LimitDegrees=15, twistLimitDegrees=8}
 {bone1Name=dyn_cape_cl_04, bone2Name=dyn_cape_cl_03, swing1LimitDegrees=10, swing2LimitDegrees=15, twistLimitDegrees=8}
 {bone1Name=dyn_cape_cl_05, bone2Name=dyn_cape_cl_04, swing1LimitDegrees=10, swing2LimitDegrees=15, twistLimitDegrees=8}
 {bone1Name=dyn_cape_cl_06, bone2Name=dyn_cape_cl_05, swing1LimitDegrees=10, swing2LimitDegrees=15, twistLimitDegrees=8}
 {bone1Name=dyn_cape_cl_08, bone2Name=dyn_cape_cl_07, swing1LimitDegrees=10, swing2LimitDegrees=15, twistLimitDegrees=8}
 {bone1Name=dyn_cape_cr_01, bone2Name=dyn_cape_cr_MSTR, swing1LimitDegrees=10, swing2LimitDegrees=30, twistLimitDegrees=8}
 {bone1Name=dyn_cape_cr_02, bone2Name=dyn_cape_cr_01, swing1LimitDegrees=10, swing2LimitDegrees=15, twistLimitDegrees=8}
 {bone1Name=dyn_cape_cr_04, bone2Name=dyn_cape_cr_03, swing1LimitDegrees=10, swing2LimitDegrees=15, twistLimitDegrees=8}
 {bone1Name=dyn_cape_cr_05, bone2Name=dyn_cape_cr_04, swing1LimitDegrees=10, swing2LimitDegrees=15, twistLimitDegrees=8}
 {bone1Name=dyn_cape_cr_06, bone2Name=dyn_cape_cr_05, swing1LimitDegrees=10, swing2LimitDegrees=15, twistLimitDegrees=8}
 {bone1Name=dyn_cape_cr_08, bone2Name=dyn_cape_cr_07, swing1LimitDegrees=10, swing2LimitDegrees=15, twistLimitDegrees=8}
 {bone1Name=dyn_cape_l_01, bone2Name=dyn_cape_l_MSTR, swing1LimitDegrees=10, swing2LimitDegrees=30, twistLimitDegrees=8}
 {bone1Name=dyn_cape_l_02, bone2Name=dyn_cape_l_01, swing1LimitDegrees=10, swing2LimitDegrees=15, twistLimitDegrees=8}
 {bone1Name=dyn_cape_l_04, bone2Name=dyn_cape_l_03, swing1LimitDegrees=10, swing2LimitDegrees=15, twistLimitDegrees=8}
 {bone1Name=dyn_cape_l_05, bone2Name=dyn_cape_l_04, swing1LimitDegrees=10, swing2LimitDegrees=15, twistLimitDegrees=8}
 {bone1Name=dyn_cape_l_07, bone2Name=dyn_cape_l_06, swing1LimitDegrees=10, swing2LimitDegrees=15, twistLimitDegrees=8}
 {bone1Name=dyn_cape_l_08, bone2Name=dyn_cape_l_07, swing1LimitDegrees=10, swing2LimitDegrees=15, twistLimitDegrees=8}
 {bone1Name=dyn_cape_r_01, bone2Name=dyn_cape_r_MSTR, swing1LimitDegrees=10, swing2LimitDegrees=30, twistLimitDegrees=8}
 {bone1Name=dyn_cape_r_02, bone2Name=dyn_cape_r_01, swing1LimitDegrees=10, swing2LimitDegrees=15, twistLimitDegrees=8}
 {bone1Name=dyn_cape_r_04, bone2Name=dyn_cape_r_03, swing1LimitDegrees=10, swing2LimitDegrees=15, twistLimitDegrees=8}
 {bone1Name=dyn_cape_r_05, bone2Name=dyn_cape_r_04, swing1LimitDegrees=10, swing2LimitDegrees=15, twistLimitDegrees=8}
 {bone1Name=dyn_cape_r_07, bone2Name=dyn_cape_r_06, swing1LimitDegrees=10, swing2LimitDegrees=15, twistLimitDegrees=8}
 {bone1Name=dyn_cape_r_08, bone2Name=dyn_cape_r_07, swing1LimitDegrees=10, swing2LimitDegrees=15, twistLimitDegrees=8}
]
