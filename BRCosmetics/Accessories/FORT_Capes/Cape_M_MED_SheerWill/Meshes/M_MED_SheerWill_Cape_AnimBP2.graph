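# Animation Blueprint for the SheerWill cape cosmetic, following the character and simulating cape motion.
# Copied parent and body poses feed collar constraints and a RigidBody node, cached for output or for fallback trail-controller chains (regular with skydiving rotation, or DBNO).
# bRigidBodyEnabled and bIsDBNO select the result.
# The event graph updates gravity, sim-space drag, grind-rail overrides and movement-based trail alpha, locates the body mesh, and toggles a material element by variant tag.
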AnimBlueprint /BRCosmetics/Accessories/FORT_Capes/Cape_M_MED_SheerWill/Meshes/M_MED_SheerWill_Cape_AnimBP2
target_skeleton=&/M_MED_SheerWill_Cape_Skeleton
graphs [
 AnimGraph {nodes [
   n1 = Use cached pose 'trailCache'()
   n2 = Use cached pose 'dbnoCache'()
   n3 = Get bIsDBNO()
   Output Pose(Result=Blend Poses by bool(
     BlendPose_0=Use cached pose 'RBANCache'()
     BlendPose_1=Blend Poses by bool(BlendPose_0=@n2, BlendPose_1=@n1, bActiveValue=@n3) {blendType=Linear}
     bActiveValue=Get bRigidBodyEnabled()
    ) {blendType=Linear})
   n4 = Layered blend per bone(
    BasePose=Copy Pose From Mesh() {bUseAttachedParent=true, rootBoneToCopy=root}
    BlendPoses_0=Copy Pose From Mesh(SourceMeshComponent=Get Character() -> (Character:Skeletal Mesh Component Budgeted Object Reference)) {bUseAttachedParent=true, rootBoneToCopy=root}
   ) {layerSetup [{branchFilters [{boneName=Collar}, {boneName=deform_upperarm_up_l}, {boneName=deform_upperarm_up_r}]}]}
   n5 = "Constraint - dyn_cape_r_MSTR (PPPPP)"!(
    ComponentPose=Local To Component(LocalPose=@n4)
    Alpha=1
    ConstraintWeights_0=0.4194
    ConstraintWeights_1:float=0.3327
    ConstraintWeights_2:float=0.244
    ConstraintWeights_3:float=0.002
    ConstraintWeights_4:float=0.0019
   ) -> (Pose) {
    boneToModify {boneName=dyn_cape_r_MSTR}
    constraintSetup [
     {targetBone {boneName=Collar}}
     {targetBone {boneName=clavicle_r}}
     {targetBone {boneName=spine_05}}
     {targetBone {boneName=spine_04}}
     {targetBone {boneName=deform_upperarm_up_r}}
    ]
    constraintWeights [1,1,1,1,1]
   }
   n6 = "Constraint - dyn_cape_cr_MSTR (P)"!(ComponentPose=@n5, Alpha=1, ConstraintWeights_0=1, ConstraintWeights_1:float=0) -> (Pose) {boneToModify {boneName=dyn_cape_cr_MSTR}, constraintSetup [{targetBone {boneName=Collar}}], constraintWeights [1,0]}
   n7 = "Constraint - dyn_cape_cl_MSTR (P)"!(ComponentPose=@n6, Alpha=1, ConstraintWeights_0=1, ConstraintWeights_1:float=0) -> (Pose) {boneToModify {boneName=dyn_cape_cl_MSTR}, constraintSetup [{targetBone {boneName=Collar}}], constraintWeights [1,0]}
   n8 = "Constraint - dyn_cape_l_MSTR (PPPPP)"!(
    ComponentPose=@n7
    Alpha=1
    ConstraintWeights_0=0.4194
    ConstraintWeights_1:float=0.3327
    ConstraintWeights_2:float=0.244
    ConstraintWeights_3:float=0.002
    ConstraintWeights_4:float=0.0019
   ) -> (Pose) {
    boneToModify {boneName=dyn_cape_l_MSTR}
    constraintSetup [
     {targetBone {boneName=Collar}}
     {targetBone {boneName=clavicle_l}}
     {targetBone {boneName=spine_05}}
     {targetBone {boneName=spine_04}}
     {targetBone {boneName=deform_upperarm_up_l}}
    ]
    constraintWeights [1,1,1,1,1]
   }
   n9 = RigidBody!(
    ComponentPose=@n8
    Alpha
    OverrideWorldGravity:Vector=Get GravityOverride()
    SimSpaceSettings:Sim Space Settings Structure=Get Sim Space Settings() -> (Sim Space Settings:Sim Space Settings Structure)
   ) -> (Pose) {
    bClampLinearTranslationLimitToRefPose=true
    bOverrideWorldGravity=true
    bUseExternalClothCollision=true
    overridePhysicsAsset=&/M_MED_SheerWill_Cape_Physics
    simulationSpace=BaseBoneSpace
    baseBoneRef {boneName=backpack_root}
    componentAppliedLinearAccClamp=<6000 6000 6000>
   }
   Save cached pose 'RBANCache'(Pose=Component To Local(ComponentPose=@n9)) {cacheName=RBANCache}
   n10 = Get SkydivingAlpha() -> (SkydivingAlpha:Rotator)
   n11 = "Transform (Modify) Bone - Bone: dyn_cape_l_01"(ComponentPose=Local To Component(LocalPose=Use cached pose 'RBANCache'()), Alpha=1, Translation="0,0,0", Rotation=@n10, Scale="1,1,1") {rotationMode=BMM_Additive, rotationSpace=BCS_BoneSpace, alphaBoolBlend {blendOption=Linear}, boneToModify {boneName=dyn_cape_l_01}}
   n12 = "Transform (Modify) Bone - Bone: dyn_cape_cr_01"(ComponentPose=@n11, Alpha=1, Translation="0,0,0", Rotation=@n10, Scale="1,1,1") {rotationMode=BMM_Additive, rotationSpace=BCS_BoneSpace, alphaBoolBlend {blendOption=Linear}, boneToModify {boneName=dyn_cape_cr_01}}
   n13 = "Transform (Modify) Bone - Bone: dyn_cape_cl_01"(ComponentPose=@n12, Alpha=1, Translation="0,0,0", Rotation=@n10, Scale="1,1,1") {rotationMode=BMM_Additive, rotationSpace=BCS_BoneSpace, alphaBoolBlend {blendOption=Linear}, boneToModify {boneName=dyn_cape_cl_01}}
   n14 = "Transform (Modify) Bone - Bone: dyn_cape_r_01"(ComponentPose=@n13, Alpha=1, Translation="0,0,0", Rotation=@n10, Scale="1,1,1") {rotationMode=BMM_Additive, rotationSpace=BCS_BoneSpace, alphaBoolBlend {blendOption=Linear}, boneToModify {boneName=dyn_cape_r_01}}
   n15 = Get BwdMovement() -> (BwdMovement:double)
   n16 = Trail controller - Bone: dyn_cape_l_08(ComponentPose=@n14, Alpha=@n15) {
    bUsePlanarLimit=true
    chainLength=7
    debugLifeTime=0.1
    maxDeltaTime=0.0333
    alphaScaleBiasClamp {bInterpResult=true}
    baseJoint {boneName=backpack_root}
    planarLimits [{drivingBone {boneName=backpack_root}, planeTransform {location=<0 0 0>, rotation=<rot 0 0 -90>, scale=<1 1 1>}}]
    rotationLimits [
     {limitMax=<0 180 180>, limitMin=<0 -180 -180>}
     {limitMax=<0 180 180>, limitMin=<0 -180 -180>}
     {limitMax=<0 180 180>, limitMin=<0 -180 -180>}
     {limitMax=<0 180 180>, limitMin=<0 -180 -180>}
     {limitMax=<0 180 180>, limitMin=<0 -180 -180>}
     {limitMax=<0 180 180>, limitMin=<0 -180 -180>}
     {limitMax=<0 180 180>, limitMin=<0 -180 -180>}
    ]
    rotationOffsets [<0 0 0>, <0 0 0>, <0 0 0>, <0 0 0>, <0 0 0>, <0 0 0>, <0 0 0>]
    trailBone {boneName=dyn_cape_l_08}
    trailRelaxationSpeed {editorCurveData {keys [<at 0 23 RCIM_Linear>,<at 0.284 19.3799 RCIM_Linear>,<at 0.6554 8.5924 RCIM_Linear>,<at 1 5 RCIM_Linear>]}}
   }
   n17 = Trail controller - Bone: dyn_cape_cl_08(ComponentPose=@n16, Alpha=@n15) {
    bUsePlanarLimit=true
    chainLength=7
    debugLifeTime=0.1
    maxDeltaTime=0.0333
    alphaScaleBiasClamp {bInterpResult=true}
    baseJoint {boneName=backpack_root}
    planarLimits [{drivingBone {boneName=backpack_root}, planeTransform {location=<0 0 0>, rotation=<rot 0 0 -90>, scale=<1 1 1>}}]
    rotationLimits [
     {limitMax=<0 180 180>, limitMin=<0 -180 -180>}
     {limitMax=<0 180 180>, limitMin=<0 -180 -180>}
     {limitMax=<0 90 180>, limitMin=<0 -90 -180>}
     {limitMax=<0 180 180>, limitMin=<0 -180 -180>}
     {limitMax=<0 180 180>, limitMin=<0 -180 -180>}
     {limitMax=<0 180 180>, limitMin=<0 -180 -180>}
     {limitMax=<0 180 180>, limitMin=<0 -180 -180>}
    ]
    rotationOffsets [<0 0 0>, <0 0 0>, <0 0 0>, <0 0 0>, <0 0 0>, <0 0 0>, <0 0 0>]
    trailBone {boneName=dyn_cape_cl_08}
    trailRelaxationSpeed {editorCurveData {keys [<at 0 23 RCIM_Linear>,<at 0.284 19.3799 RCIM_Linear>,<at 0.6554 8.5924 RCIM_Linear>,<at 1 5 RCIM_Linear>]}}
   }
   n18 = Trail controller - Bone: dyn_cape_cr_08(ComponentPose=@n17, Alpha=@n15) {
    bUsePlanarLimit=true
    chainLength=7
    debugLifeTime=0.1
    maxDeltaTime=0.0333
    alphaScaleBiasClamp {bInterpResult=true}
    baseJoint {boneName=backpack_root}
    planarLimits [{drivingBone {boneName=backpack_root}, planeTransform {location=<0 0 0>, rotation=<rot 0 0 -90>, scale=<1 1 1>}}]
    rotationLimits [
     {limitMax=<0 180 180>, limitMin=<0 -180 -180>}
     {limitMax=<0 180 180>, limitMin=<0 -180 -180>}
     {limitMax=<0 90 180>, limitMin=<0 -90 -180>}
     {limitMax=<0 180 180>, limitMin=<0 -180 -180>}
     {limitMax=<0 180 180>, limitMin=<0 -180 -180>}
     {limitMax=<0 180 180>, limitMin=<0 -180 -180>}
     {limitMax=<0 180 180>, limitMin=<0 -180 -180>}
    ]
    rotationOffsets [<0 0 0>, <0 0 0>, <0 0 0>, <0 0 0>, <0 0 0>, <0 0 0>, <0 0 0>]
    trailBone {boneName=dyn_cape_cr_08}
    trailRelaxationSpeed {editorCurveData {keys [<at 0 23 RCIM_Linear>,<at 0.284 19.3799 RCIM_Linear>,<at 0.6554 8.5924 RCIM_Linear>,<at 1 5 RCIM_Linear>]}}
   }
   n19 = Trail controller - Bone: dyn_cape_r_08(ComponentPose=@n18, Alpha=@n15) {
    bUsePlanarLimit=true
    chainLength=7
    debugLifeTime=0.1
    maxDeltaTime=0.0333
    alphaScaleBiasClamp {bInterpResult=true}
    baseJoint {boneName=backpack_root}
    planarLimits [{drivingBone {boneName=backpack_root}, planeTransform {location=<0 0 0>, rotation=<rot 0 0 -90>, scale=<1 1 1>}}]
    rotationLimits [
     {limitMax=<0 180 180>, limitMin=<0 -180 -180>}
     {limitMax=<0 180 180>, limitMin=<0 -180 -180>}
     {limitMax=<0 90 180>, limitMin=<0 -90 -180>}
     {limitMax=<0 180 180>, limitMin=<0 -180 -180>}
     {limitMax=<0 180 180>, limitMin=<0 -180 -180>}
     {limitMax=<0 180 180>, limitMin=<0 -180 -180>}
     {limitMax=<0 180 180>, limitMin=<0 -180 -180>}
    ]
    rotationOffsets [<0 0 0>, <0 0 0>, <0 0 0>, <0 0 0>, <0 0 0>, <0 0 0>, <0 0 0>]
    trailBone {boneName=dyn_cape_r_08}
    trailRelaxationSpeed {editorCurveData {keys [<at 0 23 RCIM_Linear>,<at 0.284 19.3799 RCIM_Linear>,<at 0.6554 8.5924 RCIM_Linear>,<at 1 5 RCIM_Linear>]}}
   }
   Save cached pose 'trailCache'(Pose=Component To Local(ComponentPose=@n19)) {cacheName=trailCache}
   n20 = "Transform (Modify) Bone - Bone: dyn_cape_l_01"(ComponentPose=Local To Component(LocalPose=Use cached pose 'RBANCache'()), Alpha=1, Translation="0,0,0", Rotation="-10,0,0", Scale="1,1,1") {
    rotationMode=BMM_Additive
    rotationSpace=BCS_BoneSpace
    alphaBoolBlend {blendOption=Linear}
    boneToModify {boneName=dyn_cape_l_01}
    rotation {pitch=-10}
   }
   n21 = "Transform (Modify) Bone - Bone: dyn_cape_cl_01"(ComponentPose=@n20, Alpha=1, Translation="0,0,0", Rotation="-3.475,-0.010,-0.034", Scale="1,1,1") {
    rotationMode=BMM_Additive
    rotationSpace=BCS_BoneSpace
    alphaBoolBlend {blendOption=Linear}
    boneToModify {boneName=dyn_cape_cl_01}
    rotation=<rot -3.4751 -0.01 -0.034>
   }
   n22 = "Transform (Modify) Bone - Bone: dyn_cape_cr_01"(ComponentPose=@n21, Alpha=1, Translation="0,0,0", Rotation="-3.438,-0.006,0.052", Scale="1,1,1") {
    rotationMode=BMM_Additive
    rotationSpace=BCS_BoneSpace
    alphaBoolBlend {blendOption=Linear}
    boneToModify {boneName=dyn_cape_cr_01}
    rotation=<rot -3.438 -0.006 0.052>
   }
   n23 = "Transform (Modify) Bone - Bone: dyn_cape_r_01"(ComponentPose=@n22, Alpha=1, Translation="0,0,0", Rotation="-11.964,7.908,0.743", Scale="1,1,1") {
    rotationMode=BMM_Additive
    rotationSpace=BCS_BoneSpace
    alphaBoolBlend {blendOption=Linear}
    boneToModify {boneName=dyn_cape_r_01}
    rotation=<rot -11.964 7.908 0.743>
   }
   n24 = "Transform (Modify) Bone - Bone: dyn_cape_l_03"(ComponentPose=@n23, Alpha=1, Translation="0,0,0", Rotation="-10,0,0", Scale="1,1,1") {
    rotationMode=BMM_Additive
    rotationSpace=BCS_BoneSpace
    alphaBoolBlend {blendOption=Linear}
    boneToModify {boneName=dyn_cape_l_03}
    rotation {pitch=-10}
   }
   n25 = "Transform (Modify) Bone - Bone: dyn_cape_cl_03"(ComponentPose=@n24, Alpha=1, Translation="0,0,0", Rotation="-7,0,0", Scale="1,1,1") {
    rotationMode=BMM_Additive
    rotationSpace=BCS_BoneSpace
    alphaBoolBlend {blendOption=Linear}
    boneToModify {boneName=dyn_cape_cl_03}
    rotation {pitch=-7}
   }
   n26 = "Transform (Modify) Bone - Bone: dyn_cape_cr_03"(ComponentPose=@n25, Alpha=1, Translation="0,0,0", Rotation="-7,0,0", Scale="1,1,1") {
    rotationMode=BMM_Additive
    rotationSpace=BCS_BoneSpace
    alphaBoolBlend {blendOption=Linear}
    boneToModify {boneName=dyn_cape_cr_03}
    rotation {pitch=-7}
   }
   n27 = "Transform (Modify) Bone - Bone: dyn_cape_r_03"(ComponentPose=@n26, Alpha=1, Translation="0,0,0", Rotation="-3,0,0", Scale="1,1,1") {
    rotationMode=BMM_Additive
    rotationSpace=BCS_BoneSpace
    alphaBoolBlend {blendOption=Linear}
    boneToModify {boneName=dyn_cape_r_03}
    rotation {pitch=-3}
   }
   n28 = "Transform (Modify) Bone - Bone: dyn_cape_l_04"(ComponentPose=@n27, Alpha=1, Translation="0,0,0", Rotation="-10,0,0", Scale="1,1,1") {
    rotationMode=BMM_Additive
    rotationSpace=BCS_BoneSpace
    alphaBoolBlend {blendOption=Linear}
    boneToModify {boneName=dyn_cape_l_04}
    rotation {pitch=-10}
   }
   n29 = "Transform (Modify) Bone - Bone: dyn_cape_cl_04"(ComponentPose=@n28, Alpha=1, Translation="0,0,0", Rotation="-10,0,0", Scale="1,1,1") {
    rotationMode=BMM_Additive
    rotationSpace=BCS_BoneSpace
    alphaBoolBlend {blendOption=Linear}
    boneToModify {boneName=dyn_cape_cl_04}
    rotation {pitch=-10}
   }
   n30 = "Transform (Modify) Bone - Bone: dyn_cape_cr_04"(ComponentPose=@n29, Alpha=1, Translation="0,0,0", Rotation="-10,0,0", Scale="1,1,1") {
    rotationMode=BMM_Additive
    rotationSpace=BCS_BoneSpace
    alphaBoolBlend {blendOption=Linear}
    boneToModify {boneName=dyn_cape_cr_04}
    rotation {pitch=-10}
   }
   n31 = "Transform (Modify) Bone - Bone: dyn_cape_r_04"(ComponentPose=@n30, Alpha=1, Translation="0,0,0", Rotation="-20,0,0", Scale="1,1,1") {
    rotationMode=BMM_Additive
    rotationSpace=BCS_BoneSpace
    alphaBoolBlend {blendOption=Linear}
    boneToModify {boneName=dyn_cape_r_04}
    rotation {pitch=-20}
   }
   n32 = Get BwdMovement()
   n33 = Trail controller - Bone: dyn_cape_l_08(ComponentPose=@n31, Alpha=@n32) {
    bActorSpaceFakeVel=true
    bShowTrailLocation=true
    bUsePlanarLimit=true
    chainLength=8
    debugLifeTime=0.1
    maxDeltaTime=0.0333
    relaxationSpeedScale=2
    alphaScaleBiasClamp {bInterpResult=true, interpSpeedDecreasing=25, interpSpeedIncreasing=25}
    baseJoint {boneName=backpack_root}
    planarLimits [{drivingBone {boneName=None}, planeTransform {location=<0 0 10>, rotation=<rot 0 0 0>, scale=<1 1 1>}}]
    rotationLimits [
     {limitMax=<180 180 180>}
     {limitMax=<180 180 180>}
     {limitMax=<180 180 180>}
     {limitMax=<180 180 180>}
     {limitMax=<180 180 180>}
     {limitMax=<180 180 180>}
     {limitMax=<180 180 180>}
     {limitMax=<180 180 180>}
    ]
    rotationOffsets [<0 0 0>, <0 0 0>, <0 0 0>, <0 0 0>, <0 0 0>, <0 0 0>, <0 0 0>, <0 0 0>]
    trailBone {boneName=dyn_cape_l_08}
    trailRelaxationSpeed {editorCurveData {keys [<at 0 23 RCIM_Linear>,<at 0.284 19.3799 RCIM_Linear>,<at 0.6554 8.5924 RCIM_Linear>,<at 1 5 RCIM_Linear>]}}
   }
   n34 = Trail controller - Bone: dyn_cape_cl_08(ComponentPose=@n33, Alpha=@n32) {
    bActorSpaceFakeVel=true
    bShowTrailLocation=true
    bUsePlanarLimit=true
    chainLength=8
    debugLifeTime=0.1
    maxDeltaTime=0.0333
    relaxationSpeedScale=2
    alphaScaleBiasClamp {bInterpResult=true, interpSpeedDecreasing=25, interpSpeedIncreasing=25}
    baseJoint {boneName=backpack_root}
    planarLimits [{drivingBone {boneName=dyn_cape_cl_04}, planeTransform {location=<0 0 0>, rotation=<rot 0 0 0>, scale=<1 1 1>}}]
    rotationLimits [
     {limitMax=<180 180 180>}
     {limitMax=<180 180 180>}
     {limitMax=<180 180 180>}
     {limitMax=<180 180 180>}
     {limitMax=<180 180 180>}
     {limitMax=<180 180 180>}
     {limitMax=<180 180 180>}
     {limitMax=<180 180 180>}
    ]
    rotationOffsets [<0 0 0>, <0 0 0>, <0 0 0>, <0 0 0>, <0 0 0>, <0 0 0>, <0 0 0>, <0 0 0>]
    trailBone {boneName=dyn_cape_cl_08}
    trailRelaxationSpeed {editorCurveData {keys [<at 0 23 RCIM_Linear>,<at 0.284 19.3799 RCIM_Linear>,<at 0.6554 8.5924 RCIM_Linear>,<at 1 5 RCIM_Linear>]}}
   }
   n35 = Trail controller - Bone: dyn_cape_cr_08(ComponentPose=@n34, Alpha=@n32) {
    bActorSpaceFakeVel=true
    bShowTrailLocation=true
    bUsePlanarLimit=true
    chainLength=8
    debugLifeTime=0.1
    maxDeltaTime=0.0333
    relaxationSpeedScale=2
    alphaScaleBiasClamp {bInterpResult=true, interpSpeedDecreasing=25, interpSpeedIncreasing=25}
    baseJoint {boneName=backpack_root}
    planarLimits [{drivingBone {boneName=dyn_cape_cr_04}, planeTransform {location=<0 0 0>, rotation=<rot 0 0 0>, scale=<1 1 1>}}]
    rotationLimits [
     {limitMax=<180 180 180>}
     {limitMax=<180 180 180>}
     {limitMax=<180 180 180>}
     {limitMax=<180 180 180>}
     {limitMax=<180 180 180>}
     {limitMax=<180 180 180>}
     {limitMax=<180 180 180>}
     {limitMax=<180 180 180>}
    ]
    rotationOffsets [<0 0 0>, <0 0 0>, <0 0 0>, <0 0 0>, <0 0 0>, <0 0 0>, <0 0 0>, <0 0 0>]
    trailBone {boneName=dyn_cape_cr_08}
    trailRelaxationSpeed {editorCurveData {keys [<at 0 23 RCIM_Linear>,<at 0.284 19.3799 RCIM_Linear>,<at 0.6554 8.5924 RCIM_Linear>,<at 1 5 RCIM_Linear>]}}
   }
   n36 = Trail controller - Bone: dyn_cape_r_08(ComponentPose=@n35, Alpha=@n32) {
    bActorSpaceFakeVel=true
    bShowTrailLocation=true
    bUsePlanarLimit=true
    chainLength=8
    debugLifeTime=0.1
    maxDeltaTime=0.0333
    relaxationSpeedScale=2
    alphaScaleBiasClamp {bInterpResult=true, interpSpeedDecreasing=25, interpSpeedIncreasing=25}
    baseJoint {boneName=backpack_root}
    planarLimits [{drivingBone {boneName=pelvis}, planeTransform {location=<0 0 16>, rotation=<rot 0 0 -40>, scale=<1 1 1>}}]
    rotationLimits [
     {limitMax=<180 180 180>}
     {limitMax=<180 180 180>}
     {limitMax=<180 180 180>}
     {limitMax=<180 180 180>}
     {limitMax=<180 180 180>}
     {limitMax=<180 180 180>}
     {limitMax=<180 180 180>}
     {limitMax=<180 180 180>}
    ]
    rotationOffsets [<0 0 0>, <0 0 0>, <0 0 0>, <0 0 0>, <0 0 0>, <0 0 0>, <0 0 0>, <0 0 0>]
    trailBone {boneName=dyn_cape_r_08}
    trailRelaxationSpeed {editorCurveData {keys [<at 0 23 RCIM_Linear>,<at 0.284 19.3799 RCIM_Linear>,<at 0.6554 8.5924 RCIM_Linear>,<at 1 5 RCIM_Linear>]}}
   }
   Save cached pose 'dbnoCache'(Pose=Component To Local(ComponentPose=@n36)) {cacheName=dbnoCache}
  ]}
 EventGraph {nodes [
   n1 = Event OnCharacterCustomizationLoaded()
   > Set Character(Character:Skeletal Mesh Component Budgeted Object Reference=Get Array Item with EFort Custom Part Type(Array=@n1.CustomMeshParts)) -> (Output_Get:Skeletal Mesh Component Budgeted Object Reference)
   > n2 = Get Dynamic Material Instance(SkeletalMeshComponent=GetPartSkeletalMeshComponent())
   > n3 = Branch(Condition=CharacterItemDefinitionHasAnyMetaTag(Tags=Make Array("[0]"=(TagName="Cosmetics.Variant.Property.Mat2"))))
   > SetScalarParameterValue(self=@n2, ParameterName=Hide Element 01, Value=1)
   SetScalarParameterValue(execute=@n3.else, self=@n2, ParameterName=Hide Element 01, Value=0)
   Event BlueprintUpdateAnimation()
   > n4 = Sequence!(execute) -> (then_0, then_1, then_2:Exec, then_3:Exec, then_4:Exec, then_5:Exec)
   > n5 = "Gravity Override State Function (Parameter Asset) (Nativized)"(
    CharacterPartAnimInstance=Self-Reference()
    GameTimeInSeconds=GetGameTimeInSeconds()
    DebugJointName=spine_05
    Parameters=&/M_MED_SheerWill_Cape_Parameters
   )
   > Set GravityOverride(GravityOverride=@n5)
   n6 = Break Fort Rigid Body Anim Node Parameters!(FortRigidBodyAnimNodeParameters:"Fort Rigid Body Anim Node Parameters Structure (by ref)"=@n5.OutRigidBodyAnimNodeParameters) -> (ComponentLinearAccScale:Vector=~, ComponentLinearVelScale:Vector=~, SimSpaceSettings:Sim Space Settings Structure) {is=title K2Node_BreakStruct}
   > Set Sim Space Settings(Sim Space Settings:Sim Space Settings Structure=@n6.SimSpaceSettings) -> (Output_Get:Sim Space Settings Structure)
   > n7 = PlayerMovement(AnimInstance=Self-Reference(), XDotProductCoefficient=1, YDotProductCoefficient=1, ZDotProductCoefficient=1)
   n8 = Break Sim Space Settings!(SimSpaceSettings:"Sim Space Settings Structure (by ref)"=@n6.SimSpaceSettings) -> (
    WorldAlpha:float=~
    VelocityScaleZ:float=~
    DampingAlpha:float=~
    MaxLinearVelocity:float=~
    MaxAngularVelocity:float=~
    MaxLinearAcceleration:float=~
    MaxAngularAcceleration:float=~
    ExternalLinearDragV:Vector=~
    ExternalLinearVelocity:Vector=~
    ExternalAngularVelocity:Vector=~
   ) {is=title K2Node_BreakStruct}
   n9 = MapRangeClamped(Value=@n7.YMovement, InRangeA=0, InRangeB=-1, OutRangeA=3, OutRangeB=0.002)
   > Set Sim Space Settings(Sim Space Settings=Make Sim Space Settings!(
     WorldAlpha:float=@n8
     VelocityScaleZ:float=@n8.VelocityScaleZ
     DampingAlpha:float=@n8.DampingAlpha
     MaxLinearVelocity:float=@n8.MaxLinearVelocity
     MaxAngularVelocity:float=@n8.MaxAngularVelocity
     MaxLinearAcceleration:float=@n8.MaxLinearAcceleration
     MaxAngularAcceleration:float=@n8.MaxAngularAcceleration
     ExternalLinearDragV:Vector=MakeVector(Y=@n9)
     ExternalLinearVelocity:Vector=@n8.ExternalLinearVelocity
     ExternalAngularVelocity:Vector=@n8.ExternalAngularVelocity
    ) -> (SimSpaceSettings:Sim Space Settings Structure) {is=title K2Node_MakeStruct})
   n10 = Branch(execute=@n4.then_1)
   > Set BwdMovement(BwdMovement:double=0.5) -> (Output_Get:double)
   n11 = PlayerMovement(execute=@n10.else, AnimInstance=Self-Reference(), XDotProductCoefficient=1, YDotProductCoefficient=1, ZDotProductCoefficient=1)
   > n12 = Branch(Condition=Get bIsDBNOCarried())
   > Set BwdMovement(BwdMovement=0.5)
   n13 = Branch(execute=@n12.else, Condition=AND Boolean(A=Get bIsTargeting(), B=Get bIsCrouching()))
   n14 = MapRangeClamped(
    Value=float + float(A="Absolute (Float)"(A=@n11), B=MapRangeClamped(Value=@n11.YMovement, InRangeA=-1, InRangeB=1, OutRangeA=0.3)) {is=title K2Node_CommutativeAssociativeBinaryOperator}
    InRangeA=0
    InRangeB=1.65
   )
   > Set BwdMovement(BwdMovement=float * float(A=@n14, B=0.25) {is=title K2Node_PromotableOperator})
   Set BwdMovement(execute=@n13.else, BwdMovement=@n14)
   n15 = Branch(execute=@n4.then_2, Condition=Get bIsSkyDiving())
   > n16 = Branch(Condition="float > float"(A=Get PawnVelocityZ(), B=50) {is=title K2Node_CallFunction})
   > Set BwdMovement(BwdMovement=float * float(A=Get BwdMovement(), B=0) {is=title K2Node_CommutativeAssociativeBinaryOperator})
   Set BwdMovement(execute=[@n15.else, @n16.else], BwdMovement=float * float(A=Get BwdMovement(), B=1))
   n17 = Cast To FortPlayerPawn(execute=@n4.then_3, Object=TryGetPawnOwner())
   > n18 = Branch(Condition=Get bIsOnGrindRail())
   > n19 = Branch(Condition=NOT Boolean(A=IsWeaponHolstered!(self:Fort Pawn Object Reference=@n17) -> (ReturnValue:bool)))
   > n20 = Branch(Condition=Get bIsTargeting())
   > Set Sim Space Settings(Sim Space Settings=Get Sim Space Settings_GrindTarget() -> (Sim Space Settings_GrindTarget:Sim Space Settings Structure))
   Set Sim Space Settings(
    execute=@n20.else
    Sim Space Settings=Get Sim Space Settings_GrindingWeaponActive() -> (Sim Space Settings_GrindingWeaponActive:Sim Space Settings Structure)
   )
   Set Sim Space Settings(execute=[@n18.else, @n19.else], Sim Space Settings=Get Sim Space Settings())
   n21 = GetOwningActor()
   Cast To FortPlayerMannequin!(execute=@n4.then_4, Object=@n21) -> (then, CastFailed, AsFort Player Mannequin:Fort Player Mannequin Object Reference) {is=title K2Node_DynamicCast}
   > n22 = For Each Loop(Array:Array of Skeletal Mesh Component Object References=Get Components By Class!(self:Actor Object Reference=@n21, ComponentClass:Actor Component Class Reference=/Script/Engine.SkeletalMeshComponent) -> (ReturnValue:Array of Skeletal Mesh Component Object References)) -> (Array Element:Skeletal Mesh Component Object Reference)
   > Branch(Condition="Equal (Object)"(
     A=Get SkeletalMesh!(self:Skinned Mesh Component Object Reference=@n22) -> (SkeletalMesh:Skeletal Mesh Object Reference)
     B=/BRCosmetics/Characters/Player/Male/Medium/Bodies/M_MED_SheerWill_Blot/Meshes/M_MED_SheerWill_Blot
    ))
   > n23 = Cast To SkeletalMeshComponentBudgeted!(execute, Object=@n22) -> (then, CastFailed, AsSkeletal Mesh Component Budgeted:Skeletal Mesh Component Budgeted Object Reference) {is=title K2Node_DynamicCast}
   > Set Character(Character=@n23)
   Event OnBeginSkydiving()
   > Set SkydivingAlpha(SkydivingAlpha:Rotator="65,0,0") -> (Output_Get:Rotator)
   Event OnEndSkydiving()
   > Set SkydivingAlpha(SkydivingAlpha="10,-10,0")
   TryGetPawnOwner()
  ]}
]
variables [
 BwdMovement {type=number, default=0.5}
 Character {type=SkeletalMeshComponentBudgeted}
 collar_location {type=Vector}
 collar_rotation {type=Rotator}
 GravityOverride {type=Vector}
 Sim Space Settings {
  type=SimSpaceSettings
  default {
   dampingAlpha=0
   maxAngularAcceleration=40
   maxAngularVelocity=8
   maxLinearAcceleration=4000
   maxLinearVelocity=800
   velocityScaleZ=0.7
   worldAlpha=0.5
   externalLinearDragV=<0 3 0>
  }
 }
 Sim Space Settings_GrindingWeaponActive {
  type=SimSpaceSettings
  default {
   dampingAlpha=1
   maxAngularAcceleration=10000
   maxAngularVelocity=10000
   maxLinearAcceleration=10000
   maxLinearVelocity=10000
   velocityScaleZ=1
   worldAlpha=0
   externalLinearDragV=<0 0 0>
  }
 }
 Sim Space Settings_GrindTarget {
  type=SimSpaceSettings
  default {
   dampingAlpha=1
   maxAngularAcceleration=10000
   maxAngularVelocity=10000
   maxLinearAcceleration=10000
   maxLinearVelocity=10000
   velocityScaleZ=1
   worldAlpha=0
   externalLinearDragV=<0 0 0>
  }
 }
 SkydivingAlpha {type=Rotator}
]
